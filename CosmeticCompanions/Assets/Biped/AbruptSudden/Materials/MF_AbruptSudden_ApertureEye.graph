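# Material function that draws a mechanical aperture eye for the AbruptSudden companion, closing from open to shut with a scalar driver.
# It transforms the input UVs by scale and offset, then feeds them to a six-triangle aperture function whose closing and local rotation share the Aperture Amount input.
# An offsettable soft circle mask trims the inverted aperture's edges.
# It outputs the masked eye and the circle mask separately.
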MaterialFunction /CosmeticCompanions/Assets/Biped/AbruptSudden/Materials/MF_AbruptSudden_ApertureEye
graph {nodes [
  n1 = FunctionInput(Preview=Constant2Vector()) {description=Adjust the texcoord used to position the eye aperture., inputName=Eye UV Offset, inputType=FunctionInput_Vector2, sortPriority=3}
  n2 = MaterialFunctionCall!(
   In_TextureCoordinates=FunctionInput(Preview=TextureCoordinate()) {description=Adjust the texcoord used to position the eye aperture., inputName=TexCoordinate, inputType=FunctionInput_Vector2}
   In_Scale=FunctionInput(Preview=Constant2Vector() {g=1, r=1}) {description=Adjust the scale the eye aperture., inputName=Eye UV Scale, inputType=FunctionInput_Vector2, sortPriority=4}
   In_Rotation=Constant()
   In_RotationCenter=Constant2Vector()
   In_Offset=@n1
  ) {materialFunction=/CosmeticShoes/Assets/FORT_Shoes/Shoe_SkyRattle/Materials/MF_SkyRattle_Transform}
  n3 = FunctionInput() {
   bUsePreviewValueAsDefault=true
   description=Driver value for closing the eye from 0 open to 1 closed.
   inputName=Aperture Amount
   inputType=FunctionInput_Scalar
   sortPriority=1
  }
  n4 = FunctionInput(Preview=Constant() {r=1}) {
   description=Scale size of the six triangles that create the aperture.
   inputName=Aperture Triangle Size
   inputType=FunctionInput_Scalar
   sortPriority=2
  }
  n5 = MaterialFunctionCall!(ApertureAmount=@n3, CenterPosition=@n2.Result, ApertureRotationDirection, LocalRotationAmount=@n3, TriangleScale=@n4) {materialFunction=/BRCosmetics/Characters/Player/Male/Medium/Bodies/M_MED_TungStan/Materials/MaterialFunctions/MF_TungStan_6SideAperture}
  n6 = FunctionInput(Preview=Constant2Vector()) {
   description=Adjust the texcoord used to position the circle mask.
   inputName=Eye Mask Offset
   inputType=FunctionInput_Vector2
   sortPriority=5
  }
  n7 = FunctionInput(Preview=Constant() {r=1}) {
   description=Scale of the circle that masks the edges of the aperture.
   inputName=Circle Mask Scale
   inputType=FunctionInput_Scalar
   sortPriority=6
  }
  n8 = FunctionInput(Preview=Constant() {r=0.75}) {
   description=Hardness of the falloff of the circle that masks the edges of the aperture.
   inputName=Circle Mask Hardness
   inputType=FunctionInput_Scalar
   sortPriority=7
  }
  n9 = MaterialFunctionCall!(CircleScale=@n7, CircleHardness=@n8, CenterPosition=Add(A=@n2.Result, B=@n6)) {materialFunction=/BRCosmetics/Characters/Player/Male/Medium/Bodies/M_MED_TungStan/Materials/MaterialFunctions/MF_TungStan_CircleMaskGenerator}
  FunctionOutput(None=Multiply(A=OneMinus(None=@n5.Result), B=@n9.Mask) {constA=1}) {outputName=Eye}
  FunctionOutput(None=@n9.Mask) {outputName=Mask, sortPriority=10}
 ]}
