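# Material function that layers tech-art shading for the AbruptSudden biped cosmetic companion onto incoming material attributes.
# Channel masks drive base, primary, secondary and tertiary recoloring, with optional blend colors and pattern overlay.
# Angry and worried scalars tint luminance-weighted base color and emission.
# Three procedural aperture eyes close with blink-pose parameters, the small eye following the left blink, while circle-mask gradients and noise build eye glass detail.
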MaterialFunction /CosmeticCompanions/Assets/Biped/AbruptSudden/Materials/MF_AbruptSudden_Companion_Techart
graph {nodes [
  Constant() {r=0.01}
  Constant() {r=0.01}
  n1 = SetMaterialAttributes!(
   MaterialAttributes=NamedRerouteUsage()
   BaseColor=NamedRerouteUsage()
   EmissiveColor=NamedRerouteUsage()
   Metallic=NamedRerouteUsage()
   Specular=NamedRerouteUsage()
   Roughness=NamedRerouteUsage()
  ) {attributeSetTypes [
    #1
    #2
    #3
    #4
    #5
   ]}
  FunctionOutput(None=@n1) {outputName=Base Color}
  n2 = Noise!(World Position, FilterWidth=Constant() {r=0.01})
  n3 = SmoothStep(Value=Saturate(None=Add(A=@n2) {constB=0.9})) {constMin=-0.25}
  n4 = OneMinus(None=@n3)
  n5 = NamedRerouteUsage()
  LinearInterpolate(A=@n4, B=@n3, Alpha=@n5)
  LinearInterpolate(
   A=Constant() {desc=Eye Close Value}
   B=Constant() {desc=Eye Open Value, r=0.3}
   Alpha=OneMinus(None=ScalarParameter() {group=12 Mood Expression, parameterName=Little Eye Closed Control, sliderMax=1, sortPriority=13})
  )
  n6 = TextureSampleParameter2D() {
   group=Color Customization
   parameterName=Custom Color Mask
   samplerSource=SSM_Wrap_WorldGroupSettings
   samplerType=SAMPLERTYPE_Masks
   sortPriority=11
   texture=&T_BlackMask
  }
  NamedRerouteDeclaration(None=ChannelMaskParameter(None=@n6.RGB) {group=Color Customization, parameterName=PrimaryMaskChannel, sortPriority=12}.R) {name=PrimaryMask, nodeColor=<rgba 1 0.0596 0.1687 1>}
  n7 = ChannelMaskParameter(None=@n6.RGB) {group=Color Customization, maskChannel=Green, parameterName=SecondaryMaskChannel, sortPriority=13, defaultValue=<rgba 0 1 0 0>}
  NamedRerouteDeclaration(None=@n7.R) {name=SecondaryMask, nodeColor=<rgba 0 1 0.0283 1>}
  n8 = GetMaterialAttributes(None=NamedRerouteDeclaration(None=FunctionInput() {inputName=Material, inputType=FunctionInput_MaterialAttributes}) {name=IN_Material, nodeColor=<rgba 1 0.0703 0 1>}) {attributeGetTypes [#1, #5]}
  n9 = NamedRerouteDeclaration(None=@n8.BaseColor) {name=IN_BaseColor, nodeColor=<rgba 0 0.25 1 1>}
  NamedRerouteDeclaration(None=ChannelMaskParameter(None=@n9) {group=Color Customization, parameterName=BaseColorAdvancedBlendChannel, sortPriority=18}.R) {name=AdvancedBlendMask, nodeColor=<rgba 0.5703 1 0 1>}
  n10 = ChannelMaskParameter(None=@n9) {group=Color Customization, maskChannel=Green, parameterName=BaseColorValueChannel, sortPriority=15, defaultValue=<rgba 0 1 0 0>}
  n11 = Add(A=@n10.R, B=ScalarParameter() {group=Color Customization, parameterName=BaseColorValueBias, sortPriority=16})
  NamedRerouteDeclaration(None=Power(
    Base=Saturate(None=@n11)
    Exp=ScalarParameter() {defaultValue=1, group=Color Customization, parameterName=BaseColorValuePow, sortPriority=17}
   )) {name=BaseColorValue, nodeColor=<rgba 1 0 0.8906 1>}
  n12 = VectorParameter() {group=Color Customization, parameterName=Primary Color, defaultValue=<rgba 1 1 1 1>}
  n13 = VectorParameter() {group=Color Customization, parameterName=Primary Blend Color, sortPriority=33, defaultValue=<rgba 1 1 1 1>}
  n14 = NamedRerouteUsage()
  NamedRerouteDeclaration(None=StaticSwitch(True=LinearInterpolate(A=@n12.RGB, B=@n13.RGB, Alpha=@n14), False=@n12.RGB, Value=NamedRerouteUsage())) {name=PrimaryColor, nodeColor=<rgba 0 1 0.7891 1>}
  n15 = VectorParameter() {group=Color Customization, parameterName=Secondary Color, sortPriority=34, defaultValue=<rgba 1 1 1 1>}
  n16 = VectorParameter() {group=Color Customization, parameterName=Secondary Blend Color, sortPriority=35, defaultValue=<rgba 1 1 1 1>}
  n17 = NamedRerouteUsage()
  NamedRerouteDeclaration(None=StaticSwitch(True=LinearInterpolate(A=@n15.RGB, B=@n16.RGB, Alpha=@n17), False=@n15.RGB, Value=NamedRerouteUsage())) {name=SecondaryColor, nodeColor=<rgba 0 1 0.7891 1>}
  n18 = VectorParameter() {group=Color Customization, parameterName=Tertiary Color, sortPriority=36, defaultValue=<rgba 1 1 1 1>}
  n19 = VectorParameter() {group=Color Customization, parameterName=Tertiary Blend Color, sortPriority=37, defaultValue=<rgba 1 1 1 1>}
  n20 = NamedRerouteUsage()
  NamedRerouteDeclaration(None=StaticSwitch(True=LinearInterpolate(A=@n18.RGB, B=@n19.RGB, Alpha=@n20), False=@n18.RGB, Value=NamedRerouteUsage())) {name=TertiaryColor, nodeColor=<rgba 0 1 0.7891 1>}
  n21 = TextureSampleParameter2D() {
   group=Color Customization
   parameterName=Pattern Mask
   samplerSource=SSM_Wrap_WorldGroupSettings
   samplerType=SAMPLERTYPE_Masks
   sortPriority=51
   texture=&T_BlackMask
  }
  n22 = VectorParameter() {
   desc=Put 1.0 in the channel that that the pattern mask is in and 0.0 in the other channels
   group=Color Customization
   parameterName=Pattern Mask Channel
   sortPriority=52
   defaultValue=<rgba 1 0 0 1>
  }
  NamedRerouteDeclaration(None=Saturate(None=DotProduct(A=@n21.RGB, B=@n22.RGB))) {name=PatternMask, nodeColor=<rgba 0.1719 1 0 1>}
  n23 = NamedRerouteUsage()
  n24 = VectorParameter() {group=Aperture Eye, parameterName=Left Eye Mask Offset, sortPriority=33}
  n25 = VectorParameter() {group=Aperture Eye, parameterName=Left Eye UV Offset}
  n26 = MaterialFunctionCall!(
   TexCoordinate=NamedRerouteUsage()
   Aperture Amount=NamedRerouteUsage()
   Aperture Triangle Size=NamedRerouteUsage()
   Eye UV Offset=Multiply(A=ComponentMask(None=@n25.RGB) {g=true, r=true}) {constB=-1}
   Eye UV Scale=Multiply(A=Constant2Vector() {g=1, r=1}, B=ScalarParameter() {defaultValue=1, group=Aperture Eye, parameterName=Left Eye UV Scale})
   Eye Mask Offset=Multiply(A=ComponentMask(None=@n24.RGB) {g=true, r=true}) {constB=-1}
   Circle Mask Scale=NamedRerouteUsage()
   Circle Mask Hardness=ScalarParameter() {defaultValue=0.75, group=Aperture Eye, parameterName=Left Eye Aperture Mask Hardness, sliderMax=1, sortPriority=35}
  ) {materialFunction=&/MF_AbruptSudden_ApertureEye}
  n27 = VectorParameter() {group=Aperture Eye, parameterName=Right Eye Mask Offset, sortPriority=65}
  n28 = VectorParameter() {group=Aperture Eye, parameterName=Right Eye UV Offset, sortPriority=64}
  n29 = MaterialFunctionCall(
   TexCoordinate=NamedRerouteUsage()
   Aperture Amount=NamedRerouteUsage()
   Aperture Triangle Size=NamedRerouteUsage()
   Eye UV Offset=Multiply(A=ComponentMask(None=@n28.RGB) {g=true, r=true}) {constB=-1}
   Eye UV Scale=Multiply(A=Constant2Vector() {g=1, r=1}, B=ScalarParameter() {defaultValue=1, group=Aperture Eye, parameterName=Right Eye UV Scale, sortPriority=64})
   Eye Mask Offset=Multiply(A=ComponentMask(None=@n27.RGB) {g=true, r=true}) {constB=-1}
   Circle Mask Scale=NamedRerouteUsage()
   Circle Mask Hardness=ScalarParameter() {defaultValue=0.75, group=Aperture Eye, parameterName=Right Eye Aperture Mask Hardness, sliderMax=1, sortPriority=66}
  ) {materialFunction=&MF_AbruptSudden_ApertureEye}
  n30 = VectorParameter() {group=Aperture Eye, parameterName=Small Eye UV Offset, sortPriority=128}
  n31 = VectorParameter() {group=Aperture Eye, parameterName=Small Eye Mask Offset, sortPriority=129}
  n32 = MaterialFunctionCall(
   TexCoordinate=NamedRerouteUsage()
   Aperture Amount=NamedRerouteUsage()
   Aperture Triangle Size=NamedRerouteUsage()
   Eye UV Offset=Multiply(A=ComponentMask(None=@n30.RGB) {g=true, r=true}) {constB=-1}
   Eye UV Scale=Multiply(A=Constant2Vector() {g=1, r=1}, B=ScalarParameter() {defaultValue=1, group=Aperture Eye, parameterName=Small Eye UV Scale, sortPriority=128})
   Eye Mask Offset=Multiply(A=ComponentMask(None=@n31.RGB) {g=true, r=true}) {constB=-1}
   Circle Mask Scale=ScalarParameter() {defaultValue=1, group=Aperture Eye, parameterName=Small Eye Aperture Mask Scale, sortPriority=130}
   Circle Mask Hardness=ScalarParameter() {defaultValue=0.75, group=Aperture Eye, parameterName=Small Eye Aperture Mask Hardness, sliderMax=1, sortPriority=131}
  ) {materialFunction=&MF_AbruptSudden_ApertureEye}
  MaterialExpressionAdd_23 = Add(A=Add(A=@n26.Eye, B=@n29.Eye), B=@n32.Eye)
  n33 = Convert(None=@MaterialExpressionAdd_23) {
   convertInputs [{
     type=Scalar
     expressionInput {
      expression=&self:MaterialExpressionAdd_23
      inputName=None
      mask=0
      maskA=0
      maskB=0
      maskG=0
      maskR=0
      outputIndex=0
     }
    }]
   convertOutputs [{type=Vector3}]
  }
  n34 = NamedRerouteUsage()
  n35 = NamedRerouteUsage()
  n36 = NamedRerouteUsage()
  n37 = LinearInterpolate(
   A=LinearInterpolate(A=LinearInterpolate(A=@n34, B=@n35, Alpha=@n36) {constA=1}, B=NamedRerouteUsage(), Alpha=NamedRerouteUsage())
   B=NamedRerouteUsage()
   Alpha=NamedRerouteUsage()
  )
  n38 = VectorParameter() {group=Color Customization, parameterName=Pattern Color, sortPriority=53, defaultValue=<rgba 1 1 1 1>}
  n39 = LinearInterpolate(
   A=@n37
   B=@n38.RGB
   Alpha=Multiply(A=NamedRerouteUsage(), B=ScalarParameter() {group=Color Customization, parameterName=Pattern Strength, sortPriority=54})
  ) {constA=1}
  n40 = LinearInterpolate(
   A=NamedRerouteUsage()
   B=Multiply(
    A=NamedRerouteUsage()
    B=StaticSwitchParameter(True=@n39, False=@n37) {group=Color Customization, parameterName=Use Pattern Mask, sortPriority=50}
   )
   Alpha=NamedRerouteUsage()
  )
  n41 = DotProduct(A=@n40, B=Constant3Vector() {constant=<rgba 0.2125 0.7154 0.0721 0>})
  n42 = VectorParameter() {group=12 Mood Expression, parameterName=MoodAngryColor, sortPriority=64, defaultValue=<rgba 0 1 0 1>}
  n43 = ScalarParameter() {group=12 Mood Expression, parameterName=Mood_Worried, sliderMax=1}
  n44 = ScalarParameter() {group=12 Mood Expression, parameterName=Mood_Angry, sliderMax=1}
  n45 = Multiply(A=@n44, B=OneMinus(None=@n43))
  n46 = Multiply(A=OneMinus(None=@n44), B=@n43)
  n47 = VectorParameter() {group=12 Mood Expression, parameterName=MoodWorriedColor, sortPriority=64, defaultValue=<rgba 0 0 1 1>}
  n48 = Add(A=Multiply(A=@n42.RGB, B=@n45), B=Multiply(A=@n47.RGB, B=@n46))
  n49 = ScalarParameter() {defaultValue=5, group=12 Mood Expression, parameterName=MoodColorBrightness}
  n50 = Multiply(A=Add(A=@n46, B=@n45), B=NamedRerouteUsage())
  n51 = LinearInterpolate(A=@n40, B=Multiply(A=Multiply(A=@n41, B=@n48), B=@n49), Alpha=@n50)
  n52 = Multiply(A=@n51, B=@n33)
  n53 = NamedRerouteDeclaration(None=Add(A=Add(A=@n26.Mask, B=@n29.Mask), B=@n32.Mask) {constB=0}) {name=EyeMask, nodeColor=<rgba 1 0 0.3281 1>}
  n54 = NamedRerouteUsage()
  NamedRerouteDeclaration(None=LinearInterpolate(A=@n51, B=Multiply(A=@n23, B=@n52) {constA=1}, Alpha=Subtract(A=@n53, B=@n54)) {constAlpha=0}) {name=OUT_BaseColor, nodeColor=<rgba 1 0 0.4922 1>}
  NamedRerouteDeclaration(None=ChannelMaskParameter(None=FunctionInput() {bUsePreviewValueAsDefault=true, inputName=Tech Art Mask}) {group=Color Customization, parameterName=Tech Art Mask Channel, sortPriority=10}.R) {name=Mask, nodeColor=<rgba 0 0.8125 1 1>}
  n55 = TextureCoordinate()
  NamedRerouteDeclaration(None=@n55) {name=RightEyeUV, nodeColor=<rgba 0.3125 0 1 1>}
  NamedRerouteDeclaration(None=@n55) {name=SmallEyeUV, nodeColor=<rgba 0.3125 0 1 1>}
  n56 = ChannelMaskParameter(None=@n6.RGB) {group=Color Customization, maskChannel=Blue, parameterName=TertiaryMaskChannel, sortPriority=14, defaultValue=<rgba 0 0 1 0>}
  NamedRerouteDeclaration(None=@n56.R) {name=TertiaryMask, nodeColor=<rgba 0 0.2243 1 1>}
  n57 = LinearInterpolate(
   A=Constant() {desc=Eye Close Value}
   B=Constant() {desc=Eye Open Value, r=0.3}
   Alpha=OneMinus(None=ScalarParameter() {desc=Left Eye Blink, group=12 Mood Expression, parameterName=l_blink_pose, sliderMax=1, sortPriority=11})
  )
  NamedRerouteDeclaration(None=@n57) {name=LeftEyeApertureAmount, nodeColor=<rgba 0 0.6484 1 1>}
  NamedRerouteDeclaration(None=ScalarParameter() {defaultValue=1, group=Aperture Eye, parameterName=Eye Aperture Triangle Size, sortPriority=16}) {name=TriangleSize, nodeColor=<rgba 0.9688 1 0 1>}
  NamedRerouteDeclaration(None=ScalarParameter() {defaultValue=1, group=Aperture Eye, parameterName=Eye Aperture Mask Scale, sortPriority=16}) {name=Large Eye Mask Scale, nodeColor=<rgba 1 0.1875 0 1>}
  n58 = ScalarParameter() {desc=Right Eye Blink, group=12 Mood Expression, parameterName=r_blink_pose, sliderMax=1, sortPriority=12}
  NamedRerouteDeclaration(None=LinearInterpolate(A=Constant() {desc=Eye Close Value}, B=Constant() {desc=Eye Open Value, r=0.3}, Alpha=OneMinus(None=@n58))) {name=RightEyeApertureAmount, nodeColor=<rgba 0 0.6484 1 1>}
  NamedRerouteDeclaration(None=@n57) {name=LittleEyeApertureAmount, nodeColor=<rgba 0 0.6484 1 1>}
  NamedRerouteDeclaration() {name=Eye Aperture Emission, nodeColor=<rgba 0 0.8359 1 1>}
  n59 = GetMaterialAttributes(None=NamedRerouteUsage()) {attributeGetTypes [#2, #1]}
  n60 = Multiply(A=DotProduct(A=@n59.EmissiveColor, B=Constant3Vector() {constant=<rgba 0.2125 0.7154 0.0721 0>}), B=@n48)
  NamedRerouteDeclaration(None=LinearInterpolate(A=@n59.EmissiveColor, B=Multiply(A=@n60, B=@n49), Alpha=@n50)) {name=IN Emission, nodeColor=<rgba 1 0.8438 0 1>}
  NamedRerouteDeclaration(None=@n8.Roughness) {name=Roughness, nodeColor=<rgba 0 0.2734 1 1>}
  NamedRerouteDeclaration(None=@n55) {name=LeftEyeUV, nodeColor=<rgba 0.3125 0 1 1>}
  NamedRerouteDeclaration() {name=Debug, nodeColor=<rgba 0 0.7187 1 1>}
  n61 = VertexNormalWS()
  n62 = Constant3Vector() {constant=<rgba 0 0 0.6 0>}
  n63 = DotProduct(A=Multiply(A=Add(A=@n61, B=@n62)), B=Constant3Vector() {constant=<rgba 0 0 1 0>})
  n64 = Constant() {r=0.44}
  n65 = MaterialFunctionCall!(CircleScale=Multiply(A=NamedRerouteUsage()) {constB=0.36}, CircleHardness=@n64, CenterPosition=Constant2Vector() {g=0.198, r=0.984}) {materialFunction=/BRCosmetics/Characters/Player/Male/Medium/Bodies/M_MED_TungStan/Materials/MaterialFunctions/MF_TungStan_CircleMaskGenerator}
  n66 = MaterialFunctionCall(CircleScale=Multiply(A=NamedRerouteUsage()) {constB=0.36}, CircleHardness=@n64, CenterPosition=Constant2Vector() {g=0.272, r=0.97}) {materialFunction=&MF_TungStan_CircleMaskGenerator}
  n67 = MaterialFunctionCall(
   CircleScale=Multiply(A=ScalarParameter() {defaultValue=1, group=Aperture Eye, parameterName=Small Eye Aperture Mask Scale, sortPriority=130}) {constB=0.36}
   CircleHardness=@n64
   CenterPosition=Constant2Vector() {g=0.233, r=0.9765}
  ) {materialFunction=&MF_TungStan_CircleMaskGenerator}
  n68 = Add(A=Add(A=@n65.Mask, B=@n66.Mask), B=@n67.Mask)
  n69 = Constant4Vector() {constant=<rgba 0 1 0.25 1>}
  n70 = MaterialFunctionCall(Input=Multiply(A=@n68, B=Saturate(None=@n63)), Input Low=@n69.R, Input High=@n69.R, Target Low=@n69.R, Target High=@n69.R) {materialFunction=&RemapValueRange}
  NamedRerouteDeclaration(None=Multiply(A=Saturate(None=@n70.Result), B=NamedRerouteUsage())) {name=EyeGradient, nodeColor=<rgba 0 0.2969 1 1>}
  NamedRerouteDeclaration(None=Multiply(A=@n4, B=@n5)) {name=GlassScratch, nodeColor=<rgba 0 0.2969 1 1>}
  n71 = VectorParameter() {group=Aperture Eye, parameterName=Eye Scratch Remap, sortPriority=512, defaultValue=<rgba 0.1 1.05 0 1>}
  n72 = MaterialFunctionCall(Input=@n68, Input Low=@n71.R, Input High=@n71.R, Target Low=@n71.R, Target High=@n71.R) {materialFunction=&RemapValueRange}
  n73 = Multiply(A=Saturate(None=OneMinus(None=@n72.Result)), B=NamedRerouteUsage())
  NamedRerouteDeclaration(None=Saturate(None=Multiply(A=@n73))) {name=EyeGradientInverse, nodeColor=<rgba 0.4297 0 1 1>}
  NamedRerouteDeclaration(None=@n46) {name=Mood2Select, nodeColor=<rgba 1 0 0.375 1>}
  NamedRerouteDeclaration(None=@n45) {name=Mood1Select, nodeColor=<rgba 1 0 0.375 1>}
  n74 = NamedRerouteUsage()
  n75 = Subtract(A=@n74, B=NamedRerouteUsage())
  NamedRerouteDeclaration(None=LinearInterpolate(A=@n74, B=Saturate(None=@n75), Alpha=NamedRerouteUsage())) {name=Out Rough, nodeColor=<rgba 1 0.3984 0 1>}
  NamedRerouteDeclaration(None=Multiply(A=NamedRerouteUsage(), B=NamedRerouteUsage())) {name=OUT_Emission, nodeColor=<rgba 0.6172 0 1 1>}
  n76 = GetMaterialAttributes(None=NamedRerouteUsage()) {attributeGetTypes [#3, #4]}
  n77 = NamedRerouteUsage()
  n78 = OneMinus(None=@n77)
  NamedRerouteDeclaration(None=LinearInterpolate(A=@n76.Metallic, B=Multiply(A=@n78, B=@n76.Metallic), Alpha=@n77)) {name=Out Metal, nodeColor=<rgba 0 0.6719 1 1>}
  n79 = Add(A=Multiply(A=@n77) {constB=0.22}, B=@n76.Specular)
  NamedRerouteDeclaration(None=LinearInterpolate(A=@n76.Specular, B=Saturate(None=@n79), Alpha=@n77)) {name=Out Specular, nodeColor=<rgba 0.9922 1 0 1>}
  NamedRerouteDeclaration(None=@n33) {name=Aperture, nodeColor=<rgba 0.2656 0 1 1>}
  NamedRerouteDeclaration(None=StaticBoolParameter() {group=Color Customization, parameterName=UseAdvancedColorBlend, sortPriority=20}) {name=bAdvancedColorBlend, nodeColor=<rgba 0.7109 0 1 1>}
  n80 = VectorParameter() {group=Color Customization, parameterName=Base Color, sortPriority=30, defaultValue=<rgba 1 1 1 1>}
  n81 = VectorParameter() {group=Color Customization, parameterName=Base Blend Color, sortPriority=31, defaultValue=<rgba 1 1 1 1>}
  n82 = NamedRerouteUsage()
  NamedRerouteDeclaration(None=StaticSwitch(True=LinearInterpolate(A=@n80.RGB, B=@n81.RGB, Alpha=@n82), False=@n80.RGB, Value=NamedRerouteUsage())) {name=BaseColor, nodeColor=<rgba 0 1 0.3906 1>}
  n83 = Constant()
  SetMaterialAttributes(MaterialAttributes=NamedRerouteUsage(), BaseColor=NamedRerouteUsage(), EmissiveColor=@n83, Metallic=@n83, Specular=@n83, Roughness=@n83) {attributeSetTypes [
    #1
    #2
    #3
    #4
    #5
   ]}
 ]}
