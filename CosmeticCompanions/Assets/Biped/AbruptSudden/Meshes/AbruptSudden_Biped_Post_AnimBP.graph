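# Post-process AnimBlueprint for the AbruptSudden biped cosmetic companion, adding secondary ear motion to the incoming pose.
# A constraint makes head_translate follow the head's translation, then chained AnimDynamics nodes simulate left and right ears with angular springs, gravity and wind.
# Each update scales DynamicsAlpha into separate blend weights, stronger for the left ear than the right.
# The PoseAsset blend and head spring nodes sit outside the output chain.
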AnimBlueprint /CosmeticCompanions/Assets/Biped/AbruptSudden/Meshes/AbruptSudden_Biped_Post_AnimBP
target_skeleton=&/AbruptSudden_Biped_Skeleton
graphs [
 AnimGraph {nodes [
   Component To Local(ComponentPose=Local To Component())
   Pose!(PoseAsset:Pose Asset Object Reference=Get PoseAsset() -> (PoseAsset:Pose Asset Object Reference), SourcePose:Pose Link Structure) -> (Pose:Pose Link Structure) {is=title AnimGraphNode_PoseBlendNode, bIgnorePoseAssetSourceAnimationCurves=true, blendOption=Linear}
   n1 = "Constraint - head_translate (T)"(ComponentPose=Local To Component(LocalPose=Input Pose()), Alpha=1, ConstraintWeights_0=1) {boneToModify {boneName=head_translate}, constraintSetup [{transformType=Translation, targetBone {boneName=head}}], constraintWeights [1]}
   n2 = Anim Dynamics - Bone: ear_01_l(ComponentPose=@n1, Alpha=Get Dynamics Alpha 5() -> (Dynamics Alpha 5:double), ExternalForce="0,0,0") {
    angularDampingOverride=1
    angularSpringConstant=20000
    bAngularSpring=true
    bEnableWind=true
    bOverrideAngularDamping=true
    bUseGravityOverride=true
    gravityScale=0.35
    simulationSpace=RootRelative
    windScale=0.5
    boundBone {boneName=ear_01_l}
    componentLinearAccScale=<1 1 1>
    componentLinearVelScale=<1 1 1>
    gravityOverride {z=-980}
    physicsBodyDefinitions [{
      boundBone {boneName=ear_01_l}
      boxExtents=<50 50 50>
      constraintSetup {twistAxis=AxisY, angularLimitsMax=<15 0 0>, angularLimitsMin=<-10 0 0>, angularTarget=<1 0 0>}
      localJointOffset=<0 0 0>
     }]
   }
   n3 = Anim Dynamics - Bone: ear_01_r(ComponentPose=@n2, Alpha=Get Dynamics Alpha 3() -> (Dynamics Alpha 3:double), ExternalForce="0,0,0") {
    angularDampingOverride=1
    angularSpringConstant=20000
    bAngularSpring=true
    bEnableWind=true
    bOverrideAngularDamping=true
    bUseGravityOverride=true
    gravityScale=0.35
    simulationSpace=RootRelative
    windScale=0.5
    boundBone {boneName=ear_01_r}
    componentLinearAccScale=<1 1 1>
    componentLinearVelScale=<1 1 1>
    gravityOverride {z=-980}
    physicsBodyDefinitions [{
      boundBone {boneName=ear_01_r}
      boxExtents=<60 60 60>
      constraintSetup {twistAxis=AxisY, angularLimitsMax=<10 0 0>, angularLimitsMin=<-15 0 0>, angularTarget=<1 0 0>}
      localJointOffset=<0 0 0>
     }]
   }
   Output Pose(Result=Component To Local(ComponentPose=@n3))
   Spring controller - Bone: head!(ComponentPose:Component Space Pose Link Structure, Alpha:float=0.75) -> (Pose:Component Space Pose Link Structure) {is=title AnimGraphNode_SpringBone, bTranslateX=false, bTranslateY=false, springDamping=6, springStiffness=100, springBone {boneName=head}}
  ]}
 EventGraph {nodes [
   Event BlueprintUpdateAnimation()
   > Set Dynamics Alpha 5(Dynamics Alpha 5:double=float * float(A=Get DynamicsAlpha() -> (DynamicsAlpha:float), B=0.5) {is=title K2Node_PromotableOperator}) -> (Output_Get:double)
   > Set Dynamics Alpha 3(Dynamics Alpha 3:double=float * float(A=Get DynamicsAlpha(), B=0.3)) -> (Output_Get:double)
  ]}
]
variables [
 Dynamics Alpha 3 {type=number}
 Dynamics Alpha 5 {type=number}
 PoseAsset {type=PoseAsset, default=/CosmeticCompanions/Automation_Template/companion_pose_asset_template}
]
